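; Art-tool actor that displays color breakdown modes through a post-process material, with an optional image overlay.
; The construction script looks up the selected mode's material in a map, creates a dynamic instance and sets it as the PostProcess component's weighted blendable at full weight.
; It then pushes overlay position, scale, toggle and texture into that instance's parameters.
; The BeginPlay, Tick and overlap events are empty.
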(event EventBeginPlay)

(event Collision|EventActorBeginOverlap
  (params (OtherActor :type Actor%20Object%20Reference)))

(event EventTick
  (params (DeltaSeconds :type Float%20%28single-precision%29)))

(fn SwitchDisplayMaterial
  (bind _returnvalue (Rendering|Material|CreateDynamicMaterialInstance (Utilities|Map|Find (Variables|ColorBreakdown|Internal|GetColorBreakdownModes) (Variables|ColorBreakdown|GetMode)) "None" "None" :!class /Script/Engine.KismetMaterialLibrary))
  (bind _output_get (Variables|ColorBreakdown|Internal|SetPostprocessMID _returnvalue))
  (Utilities|Struct|SetmembersinPostProcessSettings (Variables|/Script/Engine.PostProcessComponent|GetSettings (Variables|Default|GetPostProcess)) (Utilities|Struct|MakeWeightedBlendables (Utilities|Array|MakeArray (Utilities|Struct|MakeWeightedBlendable 1.0 _output_get)))))

(fn UpdateOverlayParams
  (bind (_overlayposition_x _overlayposition_y) (Math|Vector2D|BreakVector2D (Variables|GetOverlayPosition_X)))
  (bind (_overlayscale_x _overlayscale_y) (Math|Vector2D|BreakVector2D (Variables|GetOverlayScale_X)))
  (bind _postprocessmid (Variables|ColorBreakdown|Internal|GetPostprocessMID))
  ;[1 1 1] Then set up the ovelay image
  (Rendering|Material|SetVectorParameterValue _postprocessmid "OverlayParams" (Utilities|Struct|MakeLinearColor _overlayposition_x _overlayposition_y _overlayscale_x _overlayscale_y) :!class /Script/Engine.MaterialInstanceDynamic)
  (Rendering|Material|SetScalarParameterValue _postprocessmid "Overlay" (Math|Conversions|ToFloat(Boolean) (Variables|ColorBreakdown|GetAddOverlayImage)) :!class /Script/Engine.MaterialInstanceDynamic)
  (Rendering|Material|SetTextureParameterValue _postprocessmid "OverlayImage" (Variables|ColorBreakdown|Overlay|GetOverlayImage)))

(fn ConstructionScript
  (CallFunction|SwitchDisplayMaterial)
  (CallFunction|UpdateOverlayParams))
